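# EQS query for Phoebe AI bots that produces candidate navigation points across the Athena map.
# A single SimpleGrid generator lays a very large, coarsely spaced grid around the world-origin context, reused from a Satellite Salvage mission asset.
# Points are projected onto the navmesh with a tall vertical search range and the default-area recast filter.
# No tests are attached, so the query applies no scoring or filtering beyond projection.
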EnvQuery /Game/Athena/AI/Phoebe/EQS/EQS_Phoebe_BaseNavigation
options [{
  generator {
   name=EnvQueryGenerator_SimpleGrid_0
   class=EnvQueryGenerator_SimpleGrid
   properties {
    generateAround=/Game/Missions/Secondary/Events/SatelliteSalvage/Context_WorldOrigin.Context_WorldOrigin_C
    spaceBetween=3000
    gridSize {dataBinding=null, dataField=FloatValue, defaultValue=120000}
    projectionData {
     extentX=50
     navigationFilter=&RecastFilter_UseDefaultArea
     postProjectionVerticalOffset=0
     projectDown=3000
     projectUp=10000
    }
   }
  }
  tests []
 }]
