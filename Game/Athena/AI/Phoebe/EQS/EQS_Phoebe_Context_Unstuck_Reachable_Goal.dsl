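; EQS context for Phoebe AI that supplies the reachable goal location for unstuck teleport queries as a single location.
; It casts the querier to a Pawn and its AIController, then reads the AIEvaluator_UnstuckTeleport_ReachableGoal blackboard vector.
; It returns that vector if it is non-NaN and a valid AI location.
; Otherwise it returns the querier's location, logging a warning that names a different context, EQS_Phoebe_Context_NextDestinationKey.
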(fn ProvideSingleLocation
  (params (QuerierObject :type Object%20Reference) (QuerierActor :type Actor%20Object%20Reference))
  (bind _returnvalue (Utilities|Name|MakeLiteralName "AIEvaluator_UnstuckTeleport_ReachableGoal"))
  (LocalVariables|SetQuerierLocation (Transformation|GetActorLocation QuerierActor))
  (bind _aspawn (Utilities|Casting|CastToPawn QuerierActor)
    (:then
      (bind _asaicontroller (Utilities|Casting|CastToAIController (Pawn|GetController _aspawn))
        (:then
          (if (not (or (Math|Vector|VectorIsNAN (AI|Components|Blackboard|GetValueasVector (AI|GetBlackboard _asaicontroller) _returnvalue)) (not (AI|IsValidAILocation (AI|Components|Blackboard|GetValueasVector (AI|GetBlackboard _asaicontroller) _returnvalue)))))
            (return (AI|Components|Blackboard|GetValueasVector (AI|GetBlackboard _asaicontroller) _returnvalue))))
        (:CastFailed
          _)))
    (:CastFailed
      _))
  (Debugging|LogString "NewEnumerator0" "NewEnumerator0" "Can't retrieve the AIController/Pawn/GoalKey from the Querier in EQS_Phoebe_Context_NextDestinationKey" true
    (:Then
      (return (LocalVariables|GetQuerierLocation)))))
