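# EQS query for Phoebe AI that generates candidate navigable points around a world location, named for systemic points of interest.
# A grid-in-box generator centers on the world location parameter, with box width and length bound to query parameters.
# Points project onto the navmesh through a no-water filter, using a tall vertical search range.
# The query has no tests, so nothing further scores or filters the projected points.
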EnvQuery /Game/Athena/AI/Phoebe/EQS/EQS_Phoebe_SystemicPOI
options [{
  generator {
   name=FortQueryGenerator_GridInBox_0
   class=FortQueryGenerator_GridInBox
   properties {
    bAutoSortTests=true
    bCanRunAsync=false
    boxExtentsContext=null
    boxesContext=null
    envQueryResultNormalizationOption=Default
    generateAround=/Script/FortniteAI.FortQueryContext_WorldLocationParam
    minimumPointsOnLength=0
    minimumPointsOnWidth=0
    navDataOverrideContext=null
    optionName=""
    spaceBetween=1000
    boxLength {
     dataBinding=&self:FortQueryGenerator_GridInBox_0.AIDataProvider_QueryParams_1
     dataField=FloatValue
     defaultValue=1000
    }
    boxWidth {
     dataBinding=&self:FortQueryGenerator_GridInBox_0.AIDataProvider_QueryParams_0
     dataField=FloatValue
     defaultValue=1000
    }
    projectionData {
     bCanDisableTrace=true
     bCanProjectDown=true
     bCanTraceOnGeometry=true
     bCanTraceOnNavMesh=true
     bOnlyBlockingHits=true
     bTraceComplex=false
     extentX=1000
     extentY=10
     extentZ=10
     navigationFilter=/Game/Athena/AI/Phoebe/NavFilter/NavFilter_Phoebe_NoWater.NavFilter_Phoebe_NoWater_C
     postProjectionVerticalOffset=0
     projectDown=10000
     projectUp=10000
     serializedChannel=ECC_Visibility
     traceChannel=TraceTypeQuery1
     traceMode=Navigation
     traceProfileName=None
     traceShape=Line
     versionNum=1
    }
   }
  }
  tests []
 }]
